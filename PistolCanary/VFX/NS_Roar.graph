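# Niagara roar effect that pairs comic-style mesh letters with a spray of glow drops, both tinted by time-of-day color.
# ComicLetters_Weakpoint bursts nine meshes after a short delay, using execution index for mesh choice, curved row placement and staggered pop-in/out scaling, plus an outline-size attribute.
# GlowDrops combines continuous and burst sprites, pushed outward from a user offset with drag, curl noise and speed-based stretching.
# User parameters drive duration, offset and burst count.
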NiagaraSystem /PistolCanary/VFX/NS_Roar
dependencies {
 usedDataInterfaces [
  &NiagaraDataInterfaceCurve
  &NiagaraDataInterfaceArrayFloat3
  &NiagaraDataInterfaceCurlNoise
  &NiagaraDataInterfaceVector2DCurve
  &NiagaraDataInterfaceColorCurve
 ]
 usedDynamicInputs [
  &ModuloInt
  /Niagara/DynamicInputs/Math/Clamp_Vector
  &FN_FrontEndFloatModifier
  &Boolean_OrOperation
  &ColorFromCurve
  &FNTimeOfDayScalar
 ]
 usedModules [
  &self:SystemSpawnScript.SystemScriptSource.SystemScriptGraph.NiagaraNodeAssignment_0.SetVariables_#1
  &self:SystemSpawnScript.SystemScriptSource.SystemScriptGraph.NiagaraNodeAssignment_2.SetVariables_#2
  &self:SystemSpawnScript.SystemScriptSource.SystemScriptGraph.NiagaraNodeAssignment_1.SetVariables_#3
  &self:SystemSpawnScript.SystemScriptSource.SystemScriptGraph.NiagaraNodeAssignment_3.SetVariables_#4
  &self:SystemSpawnScript.SystemScriptSource.SystemScriptGraph.NiagaraNodeAssignment_4.SetVariables_#5
  &SystemState
 ]
}
emitters {
 ComicLetters_Weakpoint {
  emitterUpdateScript [
   &EmitterState(
    Inactive Response=::NewEnumerator1
    Life Cycle Mode=::NewEnumerator1
    Loop Behavior=::NewEnumerator1
    Loop Delay=0.15
    Loop Duration=$User.Duration
    UseLoopDelay=true
   )
   &SpawnBurst_Instantaneous(Spawn Count=9, Use Spawn Probability=false, UseLoopCountLimit=false)
  ]
  particleSpawnScript [
   &InitializeParticle(
    Color=&LerpLinearColors(StartColor=<rgba 1 1 1 1>, EndColor=$System.TODColor, LerpFactor=0.5)
    Lifetime=$User.Duration
    Mesh Scale Mode=::NewEnumerator3
    Mesh Uniform Scale=0.416
    Position Offset=&MakeVector(
     X=0
     Y=&Add_Float(A=88, B=&ReturnNormalizedExecIndex(Normalized Index Scale=-187.1785))
     Z=&Add_Float(
      A=100
      B=&FloatFromCurve(
       FloatCurve=DataInterface() {propertyValues {Curve {keys [
           {arriveTangent=-11.4206, interpMode=RCIM_Cubic, leaveTangent=-11.4206, tangentMode=RCTM_User, time=0.25, value=4}
           {arriveTangent=-0.4319, interpMode=RCIM_Cubic, leaveTangent=-0.4319, tangentMode=RCTM_User, time=1, value=0}
          ]}}}
       CurveIndex=&ReturnNormalizedExecIndex()
       Scale Curve=37.2974
      )
     )
    )
    Position Offset Coordinate Space=::Local
    UsePositionOffset=true
    Write Material Random=true
   )
   &InitialMeshOrientation(Mesh Orientation Mode=::NewEnumerator7, Rotation=<0 0 -0.25>, Use Rotation Vector=true)
   /SetVariables(Particles.MeshIndex=&ReturnExecIndex()) {isSetParameters=true}
   &AddVelocity(Velocity Speed Scale=-1)
   &CameraOffset(Camera Offset Amount=100)
  ]
  particleUpdateScript [
   &ParticleState(Kill Particles When Lifetime Has Elapsed=false)
   &ScaleMeshSize(Scale Factor=&VectorFromFloat(Value=&FloatFromCurve(
      FloatCurve=DataInterface() {propertyValues {Curve {keys [
          {interpMode=RCIM_Cubic, tangentMode=RCTM_User, time=0, value=0}
          {interpMode=RCIM_Cubic, tangentMode=RCTM_User, time=0.1, value=1.5}
          {arriveTangent=-1, leaveTangent=-1, time=0.2, value=1}
          {arriveTangent=0.4545, interpMode=RCIM_Cubic, leaveTangent=0.4545, time=0.6, value=1}
          {arriveTangent=0.0082, interpMode=RCIM_Cubic, leaveTangent=0.0082, tangentMode=RCTM_User, time=0.75, value=1.25}
          {interpMode=RCIM_Cubic, tangentMode=RCTM_User, time=0.8, value=0}
         ]}}}
      CurveIndex=&Subtract_Float(A=$Particles.NormalizedAge, B=&ReturnNormalizedExecIndex(Normalized Index Scale=0.2))
     )))
   &ScaleColor(
    Scale RGB=&VectorFromFloat(Value=&FloatFromCurve(
      FloatCurve=DataInterface() {propertyValues {Curve {keys [<at 0 10>,<at 0.1 1>,<at 0.5 1>]}}}
      CurveIndex=&Subtract_Float(A=$Particles.NormalizedAge, B=&ReturnNormalizedExecIndex(Normalized Index Scale=0.3))
      Scale Curve=&Multiply_Float(A=1.5, B=$System.TODFloat)
     ))
    ScaleA=true
    ScaleRGB=true
   ) {enabled=false}
   /SetVariables(Particles.OutlineSize:Vector3f=&Multiply_Vector(A=$Particles.Scale, B=<1.1095 1.05 1.05>)) {isSetParameters=true}
   &Drag(Drag=4, Use Linear Drag=true, Use Rotational Drag=true)
   &SolveForcesAndVelocity(Clamp Velocity=false, Limit Acceleration=false)
   &DynamicMaterialParameters(
    Index 0 Param 1=&FloatFromCurve(
     FloatCurve=DataInterface() {propertyValues {Curve {keys [{tangentMode=RCTM_User, time=0, value=0}, {tangentMode=RCTM_User, time=1, value=1}]}}}
     Scale Curve=40
    )
    Module.Index0Param0Write=true
    Module.Index0Param1Write=false
    Module.Index0Param2Write=false
    Module.Index0Param3Write=false
   )
  ]
  rendererClasses [&NiagaraMeshRendererProperties]
  renderers [
   {rendererClass=&NiagaraMeshRendererProperties, rendererIndex=0}
   {rendererClass=&NiagaraMeshRendererProperties, rendererIndex=1}
  ]
 }
 GlowDrops {
  emitterUpdateScript [
   &EmitterState(Scalability Mode=::NewEnumerator1)
   &SpawnRate(SpawnRate=200, Use Spawn Probability=false)
   &SpawnBurst_Instantaneous(
    Spawn Count=&MakeCustomIntFromBool(
     Bool=&SetBoolByIntComparison(A=$User.SwingSectionIndex, B=3, Comparison Type=::NewEnumerator4)
     True INT=&MakeCustomIntFromBool(Bool=$User.bIsFrontEndPreview, True INT=6, False INT=12)
     False INT=&MakeCustomIntFromBool(Bool=$User.bIsFrontEndPreview, True INT=6, False INT=5)
    )
    Spawn Time=0.04
    Use Spawn Probability=false
    UseLoopCountLimit=false
   )
  ]
  particleSpawnScript [
   &PartitionParticles(Particles Per Partition=9, Partition Mode:ENiagara_PartitionMode=::NewEnumerator1)
   &InitializeParticle(
    Color=&LerpLinearColors(StartColor=<rgba 0.4428 1 0.8068 1>, EndColor=$System.TODColor, LerpFactor=0.7)
    Lifetime Max=0.5
    Lifetime Min=0.25
    Lifetime Mode=::NewEnumerator1
    Sprite Size Mode=::NewEnumerator1
    Uniform Sprite Size Max=66
    Uniform Sprite Size Min=33
    UsePositionOffset=false
    Write Material Random=true
   )
   &SphereLocation(
    MaskSpawnGroup=false
    Offset=$User.Position Offset
    Override Seed=false
    Sphere Radius=200
    Surface Only=false
    UseNonUniformScale=false
    UseOffset=true
    UseSphereOrientation=false
   ) {module=SphereLocation001}
   &AddVelocityFromPoint(
    Module.Use Falloff Distance=false
    Origin Offset=$User.Position Offset
    Velocity Strength=&UniformRangedFloat(Minimum=350, Maximum=850)
   )
   &AddVelocity(Velocity=&Multiply_Vector(
     A=<0 0 600>
     B=&MakeCustomVectorFromBool(
      Bool=$System.LastSwing
      True Vector=&SelectVectorFromArray(
       Array Sampling Mode=::NewEnumerator0
       Direct Array Index=?
       Vector Selection Array=DataInterface() {propertyValues {FloatData [<1 1 1>, <1 -1 1>]}}
       Random Seed=0
      )
      False Vector=$System.DirectionVector
     )
    )) {enabled=false}
   &CameraOffset(Camera Offset Amount=50)
  ]
  particleUpdateScript [
   &ParticleState()
   &GravityForce(Gravity=<0 0 -50>)
   &Drag(Drag=&UniformRangedFloat(Minimum=10, Maximum=15), Use Linear Drag=true, Use Rotational Drag=false)
   &CurlNoiseForce(Noise Strength=&RandomRangeFloat(Minimum=200, Maximum=400), Pan Noise=true)
   &SolveForcesAndVelocity(Clamp Velocity=false, Limit Acceleration=false)
   &ScaleSpriteSize(Scale Factor=&Vector2DFromCurve(
     Vector2Curve=DataInterface() {propertyValues {
       XCurve {keys [<at 0 1 RCIM_Cubic>, {arriveTangent=-2.8959, interpMode=RCIM_Cubic, leaveTangent=-2.8959, tangentMode=RCTM_User, time=1, value=0}]}
       YCurve {keys [<at 0 1 RCIM_Cubic>, {arriveTangent=-2.8959, interpMode=RCIM_Cubic, leaveTangent=-2.8959, tangentMode=RCTM_User, time=1, value=0}]}
      }}
     ScaleCurve=<1 1>
    ))
   &ScaleSpriteSizeBySpeed(Max Scale Factor=<0.5 20>, Min Scale Factor=<0.8 1.2>)
   &FN_Sprite_Attributes(
    Color=<rgba 0 0 0 1>
    Create Color Attribute=true
    Create SpriteSize Attribute=true
    SpriteSize=&Multiply_Vector2D_ByFloat(Float=1.5, Vector2D=$Particles.SpriteSize)
    Write to Module Namespace=true
   )
  ]
  rendererClasses [&NiagaraSpriteRendererProperties]
  renderers [
   {rendererClass=&NiagaraSpriteRendererProperties}
   {rendererClass=&NiagaraSpriteRendererProperties, rendererIndex=1}
  ]
 }
}
summary {
 systemName=NS_Roar
 emitters [
  {emitterName=ComicLetters_Weakpoint, rendererClasses [&NiagaraMeshRendererProperties]}
  {emitterName=GlowDrops, rendererClasses [&NiagaraSpriteRendererProperties]}
 ]
 userVariables [
  User.bIsFrontEnd {type=NiagaraBool}
  User.bIsFrontEndPreview {type=NiagaraBool}
  User.Duration {type=NiagaraFloat, defaultValue=6, description=""}
  User.PhysicalSurface {type=NiagaraInt32}
  User.Position Offset {type=Vector3f, description="", defaultValue {struct=Vector3f, value=<0 50 75>}}
  User.SecondImpact {type=NiagaraBool}
  User.SwingSectionIndex {type=NiagaraInt32}
  User.WeakpointHit {type=NiagaraBool, defaultValue=true}
 ]
}
