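# Niagara system for a roar effect that draws short-lived pink speed-line sprites around the camera view.
# One CPU or GPU emitter continuously spawns long, thin sprites on a stretched ring ahead of the camera, oriented toward the system owner.
# Particles gain velocity, shift back along it and slow with drag.
# Alpha fades over life and scales by absolute camera-to-owner view alignment, so lines show when the view axis lines up with the source.
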NiagaraSystem /PistolCanary/VFX/NS_Roar_Lines
dependencies {
 usedDataInterfaces [&NiagaraDataInterfaceCamera, &NiagaraDataInterfaceCurve]
 usedDynamicInputs []
 usedModules [&SystemState]
}
emitters {LinesCamera {
  emitterUpdateScript [
   &EmitterState()
   &SpawnRate(SpawnRate=400, Use Spawn Probability=false)
  ]
  particleSpawnScript [
   &InitializeParticle(
    Color=<rgba 1 0.5518 0.9353 0.5>
    Lifetime Max=0.15
    Lifetime Min=0.12
    Lifetime Mode=::NewEnumerator1
    Position=&AddVectorToPosition(
     Position=&ReturnCameraPosition(
      Coordinate Space:ENiagaraCoordinateSpace=::Simulation
      Camera Query:NiagaraDataInterfaceCamera=DataInterface() {propertyValues {bRequireCurrentFrameData=false}}
     )
     Vector=&Multiply_VectorByFloat(
      Vector=&ReturnCameraVector(
       Camera Vectors=::NewEnumerator1
       Coordinate Space=::Simulation
       CameraQuery=DataInterface() {propertyValues {bRequireCurrentFrameData=false}}
      )
      Float=220
     )
    )
    Position Mode=::NewEnumerator1
    Sprite Size Max=<5 500>
    Sprite Size Min=<3 400>
    Sprite Size Mode=::NewEnumerator2
    UsePositionOffset=false
    Write Material Random=true
   )
   &ShapeLocation(
    MaskSpawnGroup=false
    Non Uniform Scale=<1.5 1 1>
    Override Seed=false
    Ring Radius=200
    Rotation Axis=&SubtractVectorFromPosition(Position=$Engine.Owner.Position, Vector=Linked() {linkedVariable {name=Particles.Position, type=NiagaraPosition}})
    Rotation Coordinate Space=::World
    Rotation Mode=::NewEnumerator1
    Shape Primitive=::NewEnumerator4
    UseOwnerScale=true
   )
   &AddVelocity(
    Module.Use Falloff Distance=false
    Override Seed=false
    Velocity Mode=::NewEnumerator1
    Velocity Speed=&RandomRangeFloat(Minimum=1400, Maximum=1600)
   )
   /SetVariables(Particles.Position=&AddVectorToPosition(
     Position=$Particles.Position
     Vector=&Multiply_VectorByFloat(
      Vector=&NormalizeVector(Vector To Normalize=$Particles.Velocity, Safe Normalize=true, Safe Fallback Threshold=0, Safe Fallback Vector=<0 0 0.001>)
      Float=-100
     )
    )) {isSetParameters=true}
   &CameraOffset(Camera Offset Amount=50)
  ]
  particleUpdateScript [
   &ParticleState()
   &Drag(Drag=5, Use Linear Drag=true, Use Rotational Drag=false)
   &SolveForcesAndVelocity(Clamp Velocity=false, Limit Acceleration=false)
   &ScaleColor(
    Scale Alpha=&FloatFromCurve(
     FloatCurve=DataInterface() {propertyValues {Curve {keys [
         {interpMode=RCIM_Cubic, tangentMode=RCTM_User, time=0, value=0}
         {arriveTangent=-2.0279, interpMode=RCIM_Cubic, leaveTangent=-2.0279, tangentMode=RCTM_User, time=0.2, value=1}
         {tangentMode=RCTM_User, time=1, value=0}
        ]}}}
     Scale Curve=&FloatFromCurve(
      FloatCurve=DataInterface() {propertyValues {Curve {
         defaultValue=1
         keys [
          {arriveTangentWeight=0.4006, interpMode=RCIM_Cubic, leaveTangentWeight=0.1667, tangentMode=RCTM_User, time=0.5, value=0}
          {tangentMode=RCTM_User, time=1, value=1}
         ]
        }}}
      CurveIndex=&Abs_Float(Float:NiagaraFloat=&DotProduct(
        Vector A=&NormalizeVector(
         Vector To Normalize=&Subtract_Vector(
          A=Linked() {linkedVariable {name=Engine.Owner.Position, type=NiagaraPosition}}
          B=&ReturnCameraPosition(
           Coordinate Space:ENiagaraCoordinateSpace=::World
           Camera Query:NiagaraDataInterfaceCamera=DataInterface() {propertyValues {bRequireCurrentFrameData=false}}
          )
         )
         Safe Normalize=true
         Safe Fallback Threshold=0
         Safe Fallback Vector=<0 0 0.001>
        )
        Vector B=&ReturnCameraVector(Camera Vectors=::NewEnumerator1, Coordinate Space=::World, CameraQuery=DataInterface() {propertyValues {bRequireCurrentFrameData=false}})
       ))
     )
    )
    ScaleA=true
    ScaleRGB=false
   )
   &DynamicMaterialParameters(
    Index 0 Param 1=1
    Index 0 Param 2=20
    Index 0 Param 3=1
    Module.Index0Param0Write=true
    Module.Index0Param1Write=true
    Module.Index0Param2Write=true
    Module.Index0Param3Write=true
   )
  ]
  rendererClasses [&NiagaraSpriteRendererProperties]
  renderers [{rendererClass=&NiagaraSpriteRendererProperties}]
 }}
summary {
 systemName=NS_Roar_Lines
 emitters [{emitterName=LinesCamera, rendererClasses [&NiagaraSpriteRendererProperties]}]
 userVariables []
}
